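; Valet car variant that plays special music on clients while an AI bot drives it.
; On the server, AI seat-0 entry sets AIBotIsDriving, and an AI exit that leaves that seat empty clears it.
; Its RepNotify and audio gameplay volume toggles call SetMusicStatus, which fades Music in or out and pushes or pops the Lonelies sound mix.
; Prints mention closing the car radio, but no radio logic exists here.
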(event EventBeginPlay
  (|CallParentFunction))

(event Collision|EventActorBeginOverlap
  (params (OtherActor :type Actor%20Object%20Reference)))

(event EventTick
  (params (DeltaSeconds :type Float%20%28single-precision%29)))

(event EventOnPawnEnterVehicle
  (params (PlayerPawn :type Fort%20Player%20Pawn%20Object%20Reference) (SeatIdx :type Integer))
  ;[1 1 1] On Pawn Enter Vehicle - Check if AI - If so, play specific Music.
  (|CallParentFunction PlayerPawn SeatIdx)
  (Networking|SwitchHasAuthority
    (:Authority
      (Utilities|Casting|CastToFortAthenaAIBotController (Pawn|GetController PlayerPawn))
      (if (== SeatIdx 0)
        (Variables|Default|SetAIBotIsDriving true)
        (Development|PrintString "AI just Entered the Seat - 99.999 chances there are Lonely AI - Play Custom Music.")))))

(event EventOnPawnExitVehicle
  (params (PlayerPawn :type Fort%20Player%20Pawn%20Object%20Reference) (ExitSocketName :type Name :default "None"))
  ;[1 1 1] On Pawn Exit Vehicle - Check if AI still driving or not and if not, stop specific music
  (|CallParentFunction PlayerPawn ExitSocketName)
  (Networking|SwitchHasAuthority
    (:Authority
      (Utilities|Casting|CastToFortAthenaAIBotController (Pawn|GetController PlayerPawn))
      (Utilities|IsValid (Vehicle|GetPawnatSeat 0)
        (:Is%20Valid
          (Development|PrintString "Pawn is still there"))
        (:Is%20Not%20Valid
          (Variables|Default|SetAIBotIsDriving false)
          (Development|PrintString "AI Driver got killed inthe car."))))))

(event EventPickRandomMaterial)

(event OnToggledOn(AudioGameplayVolumeProxy)
  ;[1 1 1] Comment
  (CallFunction|SetMusicStatus true))

(event OnToggledOff(AudioGameplayVolumeProxy)
  ;[1 1 1] Comment
  (CallFunction|SetMusicStatus false))

(fn OnRep_AIBotIsDriving
  (Networking|SwitchHasAuthority
    (:Remote
      (if (Variables|Default|GetAIBotIsDriving)
        (Development|PrintString "Close normal Car Radio and Start Special Unique Music for Henchbros")
        (CallFunction|SetMusicStatus true)
        (else
          (Development|PrintString "Kill Special Henchbros Music")
          (CallFunction|SetMusicStatus false))))))

(fn SetMusicStatus
  (params (ShouldPlay :type Boolean))
  (if (Variables|Default|GetAIBotIsDriving)
    (if ShouldPlay
      (Audio|Components|Audio|FadeIn (Variables|Default|GetMusic) 1.5 1.0 0.0 "Linear" :!class /Script/Engine.AudioComponent)
      (Variables|Default|GetIn%20Sound%20Mix%20Modifier
        (:then)
        (:else
          (Audio|PushSoundMixModifier "/Lonelies/Audio/Lonelies_MixMod.Lonelies_MixMod")
          (Variables|Default|SetIn%20Sound%20Mix%20Modifier "/Lonelies/Audio/Lonelies_MixMod.Lonelies_MixMod")))
      (else
        _))
    (else
      _))
  (Audio|PopSoundMixModifier "/Lonelies/Audio/Lonelies_MixMod.Lonelies_MixMod")
  (Variables|Default|SetIn%20Sound%20Mix%20Modifier 0)
  (if (Audio|Components|Audio|IsPlaying (Variables|Default|GetMusic))
    (Audio|Components|Audio|FadeOut (Variables|Default|GetMusic) 1.5 0.0 "Linear" :!class /Script/Engine.AudioComponent)))

(fn ConstructionScript
  (|CallParentFunction))
